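; Quest gameplay ability that broadcasts the AllSlotsUpgraded custom verb when the player collects or holds a fully modded weapon.
; On activation it listens for the CollectedPickup event and WeaponMod tag additions, passing the pickup or current weapon to CheckMods.
; CheckMods counts mods, excluding default iron sights and magazine by object name, prints a debug total, and broadcasts at four or more.
; TimerCheck is marked old, though EndAbility still clears its timer handle.
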(event Custom|CheckMods
  (params (Object :type Object%20Reference))
  (Variables|Default|SetModCount 0)
  (bind _asfort_pickup_athena (Utilities|Casting|CastToFortPickupAthena Object)
    (:then
      (bind _returnvalue (Class|FortWeaponModFunctionLibrary|GetWeaponModsfromPickup _asfort_pickup_athena))
      (for _array_element _returnvalue
        (if (and (Utilities|String|NotEqual,CaseInsenstive(String) "WMID_Default_IronSights" (Utilities|GetObjectName _array_element)) (Utilities|String|NotEqual,CaseInsenstive(String) (Utilities|GetObjectName _array_element) "WMID_Default_Magazine"))
          (Math|Integer|IncrementInt (Variables|Default|GetModCount))))
      (Development|PrintString (Utilities|String|Append "TOTAL MODS: " (Utilities|String|ToString(Integer) (Variables|Default|GetModCount))) false false)
      (if (>= (Variables|Default|GetModCount) 4)
        (CustomVerb|BroadcastCustomVerbMessage (CustomVerb|MakeCustomVerbMessage (Ability|GetAvatarActorfromActorInfo) "(TagName=\"CustomVerb.WeaponMods.AllSlotsUpgraded\")"))))
    (:CastFailed
      (bind _asfort_weapon (Utilities|Casting|CastToFortWeapon Object))
      (for _array_element_2 (Class|FortWeapon|GetWeaponModSlots _asfort_weapon)
        (if (and (Utilities|String|NotEqual,CaseInsenstive(String) "WMID_Default_IronSights" (Utilities|GetObjectName (Utilities|Struct|BreakFortWeaponModSlot _array_element_2))) (Utilities|String|NotEqual,CaseInsenstive(String) (Utilities|GetObjectName (Utilities|Struct|BreakFortWeaponModSlot _array_element_2)) "WMID_Default_Magazine"))
          (Math|Integer|IncrementInt (Variables|Default|GetModCount))))
      (Development|PrintString (Utilities|String|Append "TOTAL MODS: " (Utilities|String|ToString(Integer) (Variables|Default|GetModCount))) false false)
      (if (>= (Variables|Default|GetModCount) 4)
        (CustomVerb|BroadcastCustomVerbMessage (CustomVerb|MakeCustomVerbMessage (Ability|GetAvatarActorfromActorInfo) "(TagName=\"CustomVerb.WeaponMods.AllSlotsUpgraded\")"))))))

(event Ability|EventActivateAbility
  (bind _returnvalue (Ability|GetAvatarActorfromActorInfo))
  (Utilities|FlowControl|Sequence
    (:then_0
      (bind (_asynctaskproxy _payload) (Ability|Async|WaitGameplayEventToActor _returnvalue "(TagName=\"Event.CollectedPickup\")")
        (:EventReceived
          (CallFunction|CheckMods (.Target (Utilities|Struct|BreakGameplayEventData _payload))))))
    (:then_1
      (Ability|Tasks|WaitGameplayTagAddToActor _returnvalue "(TagName=\"WeaponMod\")"
        (:Added
          (CallFunction|CheckMods (Weapon|GetCurrentWeapon (Game|Player|GetPlayerPawn (Fortnite|GetFortPlayerControllerfromActor (Ability|GetAvatarActorfromActorInfo))))))))
    (:then_2)
    (:then_3)))

(event Custom|TimerCheck
  ;[1 1 1] OLD
  (CallFunction|CheckMods (Weapon|GetCurrentWeapon (Game|Player|GetPlayerPawn (Fortnite|GetFortPlayerControllerfromActor (Ability|GetAvatarActorfromActorInfo))))))

(event Ability|EventOnEndAbility
  (params (bWasCancelled :type Boolean))
  (if (Utilities|Time|IsValidTimerHandle (Variables|Default|GetWorkbenchTimer))
    (Utilities|Time|ClearandInvalidateTimerbyHandle (Variables|Default|GetWorkbenchTimer))))
